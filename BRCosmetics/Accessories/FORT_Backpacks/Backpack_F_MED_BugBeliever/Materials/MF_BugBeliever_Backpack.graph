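# Material function for the BugBeliever backpack that adjusts emissive brightness by time of day.
# It breaks out the incoming material attributes, and four scalar parameters set morning, day, evening and night emissive amounts.
# Scalar_Time_of_Day resolves these into one multiplier for the emissive total, which is written back through SetMaterialAttributes.
# Named reroutes carry the values between stages, and a tech art mask input is also exposed.
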MaterialFunction /BRCosmetics/Accessories/FORT_Backpacks/Backpack_F_MED_BugBeliever/Materials/MF_BugBeliever_Backpack
graph {
 properties {bExposeToLibrary=true, description=BugBeliever backpack tech art extras, userExposedCaption=BugBeliever Backpack Tech Art}
 nodes [
  FunctionOutput(None=SetMaterialAttributes!(MaterialAttributes=NamedRerouteUsage(), EmissiveColor=NamedRerouteUsage()) {attributeSetTypes [#1]}) {outputName=OUT_MaterialAttribute}
  NamedRerouteDeclaration(None=FunctionInput() {inputName=IN_TechArtMask, sortPriority=1}) {name=IN_TechArtMask, nodeColor=<rgba 0.4531 1 0 1>}
  n1 = GetMaterialAttributes(None=NamedRerouteDeclaration(None=FunctionInput() {bUsePreviewValueAsDefault=true, inputName=IN_MaterialAttribute, inputType=FunctionInput_MaterialAttributes}) {name=IN_MaterialAttribute, nodeColor=<rgba 0 0.1328 1 1>}) {attributeGetTypes [#1, #2]}
  NamedRerouteDeclaration(None=@n1.EmissiveColor) {name=IN_EmissiveColor, nodeColor=<rgba 0.7344 0 1 1>}
  NamedRerouteDeclaration(None=@n1.BaseColor) {name=IN_BaseColor, nodeColor=<rgba 0 1 0.4141 1>}
  NamedRerouteDeclaration(None=ScalarParameter() {defaultValue=1.1, group=BugBelieverBackpackTechArt, parameterName=EmissiveMorningAmount}) {name=EmissiveMorningAmount, nodeColor=<rgba 1 0.0938 0 1>}
  NamedRerouteDeclaration(None=ScalarParameter() {defaultValue=1, group=BugBelieverBackpackTechArt, parameterName=EmissiveDayAmount}) {name=EmissiveDayAmount, nodeColor=<rgba 0 0.2266 1 1>}
  NamedRerouteDeclaration(None=ScalarParameter() {defaultValue=1.5, group=BugBelieverBackpackTechArt, parameterName=EmissiveEveningAmount}) {name=EmissiveEveningAmount, nodeColor=<rgba 0.5469 1 0 1>}
  NamedRerouteDeclaration(None=ScalarParameter() {defaultValue=3, group=BugBelieverBackpackTechArt, parameterName=EmissiveNightAmount}) {name=EmissiveNightAmount, nodeColor=<rgba 1 0 0.8672 1>}
  n2 = NamedRerouteUsage()
  n3 = NamedRerouteUsage()
  n4 = NamedRerouteUsage()
  n5 = NamedRerouteUsage()
  NamedRerouteDeclaration(None=Multiply(
    A=NamedRerouteUsage()
    B=MaterialFunctionCall(Value_Morning=@n2, Value_Day=@n3, Value_Evening=@n4, Value_Night=@n5) {materialFunction=&Scalar_Time_of_Day}.Result
   )) {name=EmissiveColorTotal, nodeColor=<rgba 0 1 0.8125 1>}
 ]
}
