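# Control Rig for the BugBeliever backpack cosmetic that makes the bug's center eye glance around at random.
# Timed random floats pick one of nine look directions (combinations of Y and Z limits, or center) and scale its amount, with the hold duration itself randomized.
# An accumulated lerp smooths the chosen Euler angles, which rotate the center eye control, and the eye bone is snapped to that control.
# Wings, legs, pincers and other eyes are not driven here.
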ControlRigBlueprint /BRCosmetics/Accessories/FORT_Backpacks/Backpack_F_MED_BugBeliever/Meshes/F_MED_BugBeliever_Pack_CtrlRig
graphs [
 RigVMFunctionLibrary {links [], nodes []}
 RigVMModel {nodes [
   RigUnit_BeginExecution = &RigUnit_BeginExecution()
   n1 = /Script/RigVM.RigVMFunction_RandomFloat(Duration=2, Maximum=2, Minimum=0.5, Seed=217) {is=struct RigVMUnitNode, template="Random::Execute(in Seed,in Minimum,in Maximum,in Duration,out Result)"}
   RigVMFunction_MathFloatMake = /Script/RigVM.RigVMFunction_MathFloatMake(Value=&RigVMFunction_RandomFloat(Maximum=8, Minimum=0, Seed=115, Duration=@n1.Result) {template="Random::Execute(in Seed,in Minimum,in Maximum,in Duration,out Result)"}.Result) {is=struct RigVMUnitNode}
   RigVMFunction_MathFloatToInt = /Script/RigVM.RigVMFunction_MathFloatToInt(Value=@RigVMFunction_MathFloatMake.Value) {is=struct RigVMUnitNode, template="Cast::Execute(in Value,out Result)"}
   RigVMFunction_MathIntMake = /Script/RigVM.RigVMFunction_MathIntMake(Value=@RigVMFunction_MathFloatToInt.Result) {is=struct RigVMUnitNode}
   n2 = &RigVMFunction_RandomFloat(Maximum=1, Minimum=0.25, Seed=217, Duration=@n1.Result) {template="Random::Execute(in Seed,in Minimum,in Maximum,in Duration,out Result)"}
   n3 = &RigVMFunction_MathDoubleMul(B=@n2.Result, A=RigVMVariableNode() {variable=maxY}.Value)
   n4 = &RigVMFunction_MathDoubleMul(B=@n2.Result, A=RigVMVariableNode() {variable=minY}.Value)
   n5 = &RigVMFunction_RandomFloat(Maximum=1, Minimum=0.5, Seed=150, Duration=@n1.Result) {template="Random::Execute(in Seed,in Minimum,in Maximum,in Duration,out Result)"}
   n6 = &RigVMFunction_MathDoubleMul(B=@n5.Result, A=RigVMVariableNode() {variable=minZ}.Value)
   n7 = &RigVMFunction_MathDoubleMul(B=@n5.Result, A=RigVMVariableNode() {variable=maxZ}.Value)
   n8 = RigVMDispatchNode(
    Values="((X=0,Y=0,Z=0),(X=0,Y=0,Z=0),((X=0,Y=0,Z=0),(X=0,Y=0,Z=0),X=0,Y=0,Z=0),((X=0,Y=0,Z=0),(X=0,Y=0,Z=0),X=0,Y=0,Z=0),((X=0,Y=0,Z=0),(X=0,Y=0,Z=0),X=0,Y=0,Z=0),((X=0,Y=0,Z=0),(X=0,Y=0,Z=0),X=0,Y=0,Z=0),((X=0,Y=0,Z=0),(X=0,Y=0,Z=0),X=0,Y=0,Z=0),((X=0,Y=0,Z=0),(X ...(truncated)"
    Index=@RigVMFunction_MathIntMake.Value
    Values.0=/Script/RigVM.RigVMFunction_MathVectorMake(X=0, Z=0, Y=@n3.Result) {is=struct RigVMUnitNode}.Result
    Values.1=&RigVMFunction_MathVectorMake(X=0, Z=0, Y=@n4.Result).Result
    Values.2=&RigVMFunction_MathVectorMake(X=0, Y=0, Z=0).Result
    Values.3=&RigVMFunction_MathVectorMake(X=0, Y=0, Z=@n6.Result).Result
    Values.4=&RigVMFunction_MathVectorMake(X=0, Y=0, Z=@n7.Result).Result
    Values.5=&RigVMFunction_MathVectorMake(X=0, Y=@n4.Result, Z=@n7.Result).Result
    Values.6=&RigVMFunction_MathVectorMake(X=0, Y=@n4.Result, Z=@n6.Result).Result
    Values.8=&RigVMFunction_MathVectorMake(X=0, Y=@n3.Result, Z=@n7.Result).Result
    Values.7=&RigVMFunction_MathVectorMake(X=0, Y=@n3.Result, Z=@n6.Result).Result
   ) {template="DISPATCH_RigVMDispatch_SelectInt32(in Index,in Values,out Result)"}
   RigVMFunction_MathQuaternionFromEuler = &RigVMFunction_MathQuaternionFromEuler(
    RotationOrder=XYZ
    Euler=/Script/RigVM.RigVMFunction_AccumulateVectorLerp(Blend=0.25, InitialValue=(X=0,Y=0,Z=0), bIntegrateDeltaTime=false, TargetValue=@n8.Result) {is=struct RigVMUnitNode, template="AccumulateLerp::Execute(in TargetValue,in InitialValue,in Blend,in bIntegrateDeltaTime,out Result)"}.Result
   )
   RigUnit_SetRotation = &RigUnit_SetRotation(
    Item=(Type=Control,Name="dyn_bug_center_eye_01_ctrl")
    Weight=1
    bInitial=False
    ExecutePin=@RigUnit_BeginExecution.ExecutePin
    Value=@RigVMFunction_MathQuaternionFromEuler.Result
   )
   > RigUnit_SetRelativeTransformForItem = /Script/ControlRig.RigUnit_SetRelativeTransformForItem(
    Child=(Type=Bone,Name="dyn_bug_center_eye_01")
    Parent=(Type=Control,Name="dyn_bug_center_eye_01_ctrl")
    Value=(Rotation=(X=0,Y=0,Z=0,W=1),Translation=(X=0,Y=0,Z=0),Scale3D=(X=1,Y=1,Z=1))
    Weight=1
    bParentInitial=False
    bPropagateToChildren=True
   ) {
    is=struct RigVMUnitNode
    template="Set Relative Transform::Execute(in Child,in Parent,in bParentInitial,in Value,in Weight,in bPropagateToChildren)"
   }
  ]}
]
hierarchy [
 root {loc=<0 0 0>, rot=<0 0 0 1>}
 pack {parent=root, loc=<0 -0.7316 0>, rot=<0 0 0 1>}
 Bug_center {parent=pack, loc=<0 -6.1027 -3.9122>, rot=<0 0 0 1>}
 dyn_bug_wing_up_l_base {parent=Bug_center, loc=<1.7722 -7.6812 -0.8212>, rot=<-0.0109 0.164 0.2486 -0.9545>}
 dyn_bug_wing_up_l_01 {parent=dyn_bug_wing_up_l_base, loc=<0 0 0>, rot=<0 0 0 1>}
 dyn_bug_wing_mid_l_base {parent=Bug_center, loc=<2.5678 -6.7937 -4.0595>, rot=<-0.1885 0.0808 0.308 -0.929>}
 dyn_bug_wing_mid_l_01 {parent=dyn_bug_wing_mid_l_base, loc=<0 0 0>, rot=<0 0 0 1>}
 dyn_bug_wing_up_r_base {parent=Bug_center, loc=<-1.7722 -7.6812 -0.8212>, rot=<0.164 -0.0109 0.9545 -0.2486>}
 dyn_bug_wing_up_r_01 {parent=dyn_bug_wing_up_r_base, loc=<0 0 0>, rot=<0 0 0 1>}
 dyn_bug_wing_mid_r_base {parent=Bug_center, loc=<-2.5678 -6.7937 -4.0595>, rot=<0.0808 -0.1885 0.929 -0.308>}
 dyn_bug_wing_mid_r_01 {parent=dyn_bug_wing_mid_r_base, loc=<0 0 0>, rot=<0 0 0 1>}
 dyn_bug_leg_up_l_base {parent=Bug_center, loc=<9.3882 3.2049 2.828>, rot=<0 0.1476 0 -0.989>}
 dyn_bug_leg_up_l_01 {parent=dyn_bug_leg_up_l_base, loc=<0 0 0>, rot=<0 0 0 1>}
 dyn_bug_leg_mid_l_base {parent=Bug_center, loc=<9.752 2.968 -1.7324>, rot=<0 -0.0506 0 -0.9987>}
 dyn_bug_leg_mid_l_01 {parent=dyn_bug_leg_mid_l_base, loc=<0 0 0>, rot=<0 0 0 1>}
 dyn_bug_leg_bottom_l_base {parent=Bug_center, loc=<8.5578 3.3305 -6.0643>, rot=<0 -0.2076 0 -0.9782>}
 dyn_bug_leg_bottom_l_01 {parent=dyn_bug_leg_bottom_l_base, loc=<0 0 0>, rot=<0 0 0 1>}
 dyn_bug_leg_up_r_base {parent=Bug_center, loc=<-9.3882 3.2049 2.828>, rot=<0.1476 0 0.989 0>}
 dyn_bug_leg_up_r_01 {parent=dyn_bug_leg_up_r_base, loc=<0 0 0>, rot=<0 0 0 1>}
 dyn_bug_leg_mid_r_base {parent=Bug_center, loc=<-9.752 2.968 -1.7324>, rot=<-0.0506 0 0.9987 0>}
 dyn_bug_leg_mid_r_01 {parent=dyn_bug_leg_mid_r_base, loc=<0 0 0>, rot=<0 0 0 1>}
 dyn_bug_leg_bottom_r_base {parent=Bug_center, loc=<-8.5578 3.3305 -6.0643>, rot=<-0.2076 0 0.9782 0>}
 dyn_bug_leg_bottom_r_01 {parent=dyn_bug_leg_bottom_r_base, loc=<0 0 0>, rot=<0 0 0 1>}
 dyn_bug_pincer_l_base {parent=Bug_center, loc=<3.9904 3.0655 9.26>, rot=<0.0326 0.5551 -0.0487 -0.8297>}
 dyn_bug_pincer_l_01 {parent=dyn_bug_pincer_l_base, loc=<0 0 0>, rot=<0 0 0 1>}
 dyn_bug_pincer_r_base {parent=Bug_center, loc=<-3.9904 3.0655 9.26>, rot=<0.5551 0.0326 0.8297 0.0487>}
 dyn_bug_pincer_r_01 {parent=dyn_bug_pincer_r_base, loc=<0 0 0>, rot=<0 0 0 1>}
 dyn_bug_center_eye_base {parent=Bug_center, loc=<0 -2.8477 2.0347>, rot=<0.1858 0.1858 0.6823 -0.6823>}
 dyn_bug_center_eye_01 {parent=dyn_bug_center_eye_base, loc=<0 0 0>, rot=<0 0 0 1>}
 dyn_bug_center_eye_02 {parent=dyn_bug_center_eye_01, loc=<6.0109 0 0>, rot=<0 0 0 1>}
 dyn_bug_center_eye_01_ctrl {controlType=EULER_TRANSFORM, parent=dyn_bug_center_eye_base, type=Control, loc=<0 0 0>, rot=<0 0 0 1>}
 dyn_bug_mid_eye_l_base {parent=Bug_center, loc=<1.9633 -2.1591 8.1911>, rot=<0.3 0.1658 0.6869 -0.6409>}
 dyn_bug_mid_eye_l_01 {parent=dyn_bug_mid_eye_l_base, loc=<0 0 0>, rot=<0 0 0 1>}
 dyn_bug_mid_eye_l_02 {parent=dyn_bug_mid_eye_l_01, loc=<1.3949 0 0>, rot=<0 0 0 1>}
 dyn_bug_mid_eye_r_base {parent=Bug_center, loc=<-1.9662 -2.1591 8.1911>, rot=<0.1663 0.2994 0.6406 -0.6873>}
 dyn_bug_mid_eye_r_01 {parent=dyn_bug_mid_eye_r_base, loc=<0 0 0>, rot=<0 0 0 1>}
 dyn_bug_mid_eye_r_02 {parent=dyn_bug_mid_eye_r_01, loc=<1.3987 0 0>, rot=<0 0 0 1>}
 dyn_bug_side_eye_r_base {parent=Bug_center, loc=<-5.1295 -2.3132 6.8259>, rot=<-0.0322 0.3737 0.6895 -0.6196>}
 dyn_bug_side_eye_r_01 {parent=dyn_bug_side_eye_r_base, loc=<0 0 0>, rot=<0 0 0 1>}
 dyn_bug_side_eye_r_02 {parent=dyn_bug_side_eye_r_01, loc=<1.1107 0 0>, rot=<0 0 0 1>}
 dyn_bug_side_eye_l_base {parent=Bug_center, loc=<5.063 -2.2667 6.8041>, rot=<0.3411 -0.0074 0.6081 -0.7168>}
 dyn_bug_side_eye_l_01 {parent=dyn_bug_side_eye_l_base, loc=<0 0 0>, rot=<0 0 0 1>}
 dyn_bug_side_eye_l_02 {parent=dyn_bug_side_eye_l_01, loc=<1.1799 0 0>, rot=<0 0 0 1>}
 dyn_bug_zippo_base {parent=Bug_center, loc=<0 -8.1442 -2.1357>, rot=<0.0045 -0.7071 0.0045 -0.7071>}
 dyn_bug_zippo_01 {parent=dyn_bug_zippo_base, loc=<0 0 0>, rot=<0 0 0 1>}
]
variables [
 Speed {type=double, default=2}
 Time {type=double, default=0}
 maxY {type=double, default=22}
 maxZ {type=double, default=22}
 minY {type=double, default=-22}
 minZ {type=double, default=-22}
]
