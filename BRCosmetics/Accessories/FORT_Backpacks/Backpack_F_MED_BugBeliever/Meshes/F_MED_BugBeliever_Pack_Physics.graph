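# Physics asset for the BugBeliever cosmetic backpack, giving the bug's legs, eyes, pincers and wings capsule bodies for secondary motion.
# Each appendage chains a capsule to a zero-size base body, which joins the central Bug_center capsule with free swing and twist.
# The outer joints carry per-part swing limits and little or no twist, with pincers and wings generally tighter than legs.
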PhysicsAsset /BRCosmetics/Accessories/FORT_Backpacks/Backpack_F_MED_BugBeliever/Meshes/F_MED_BugBeliever_Pack_Physics
bodies {
 Bug_center [{shapeType=Capsule, length=4, radius=2, shapeName=Bug_center_capsule, rotation=<rot 90 0 0>}]
 dyn_bug_center_eye_01 [{shapeType=Capsule, length=3.5829, radius=0.505, shapeName=dyn_bug_center_eye_01_capsule, center=<3 0 0>, rotation=<rot 0 -90 90>}]
 dyn_bug_center_eye_base [{shapeType=Capsule, length=0, radius=0, shapeName=dyn_bug_center_eye_base_capsule, rotation=<rot 90 0 0>}]
 dyn_bug_leg_bottom_l_01 [{
   shapeType=Capsule
   length=2.256
   radius=2
   shapeName=dyn_bug_leg_bottom_l_01_capsule
   center=<2.1219 1.1278 1.0379>
   rotation=<rot 47.159 -102.0523 -179.9999>
  }]
 dyn_bug_leg_bottom_l_base [{shapeType=Capsule, length=0, radius=0, shapeName=dyn_bug_leg_bottom_l_base_capsule, rotation=<rot 90 0 0>}]
 dyn_bug_leg_bottom_r_01 [{
   shapeType=Capsule
   length=2.256
   radius=2
   shapeName=dyn_bug_leg_bottom_r_01_capsule
   center=<1.005 -2.0322 0.8066>
   rotation=<rot -65.1904 -102.0522 19.2044>
  }]
 dyn_bug_leg_bottom_r_base [{shapeType=Capsule, length=0, radius=0, shapeName=dyn_bug_leg_bottom_r_base_capsule, rotation=<rot 90 0 0>}]
 dyn_bug_leg_mid_l_01 [{
   shapeType=Capsule
   length=2.256
   radius=2
   shapeName=dyn_bug_leg_mid_l_01_capsule
   center=<2.1219 1.1278 1.0379>
   rotation=<rot 54.7696 -102.0524 -179.9999>
  }]
 dyn_bug_leg_mid_l_base [{shapeType=Capsule, length=0, radius=0, shapeName=dyn_bug_leg_mid_l_base_capsule, rotation=<rot 90 0 0>}]
 dyn_bug_leg_mid_r_01 [{
   shapeType=Capsule
   length=2.256
   radius=2
   shapeName=dyn_bug_leg_mid_r_01_capsule
   center=<1.8074 -1.3787 1.0849>
   rotation=<rot 44.4065 113.0566 -0.0022>
  }]
 dyn_bug_leg_mid_r_base [{shapeType=Capsule, length=0, radius=0, shapeName=dyn_bug_leg_mid_r_base_capsule, rotation=<rot 90 0 0>}]
 dyn_bug_leg_up_l_01 [{
   shapeType=Capsule
   length=2.256
   radius=2
   shapeName=dyn_bug_leg_up_l_01_capsule
   center=<2.1219 1.1278 1.0379>
   rotation=<rot 54.7827 -102.0527 -179.9999>
  }]
 dyn_bug_leg_up_l_base [{shapeType=Capsule, length=0, radius=0, shapeName=dyn_bug_leg_up_l_base_capsule, rotation=<rot 90 0 0>}]
 dyn_bug_leg_up_r_01 [{
   shapeType=Capsule
   length=2.256
   radius=2
   shapeName=dyn_bug_leg_up_r_01_capsule
   center=<1.7238 -1.3608 0.8134>
   rotation=<rot 49.6694 113.0585 179.9997>
  }]
 dyn_bug_leg_up_r_base [{shapeType=Capsule, length=0, radius=0, shapeName=dyn_bug_leg_up_r_base_capsule, rotation=<rot 90 0 0>}]
 dyn_bug_mid_eye_l_01 [{shapeType=Capsule, length=0.6624, radius=1.6169, shapeName=dyn_bug_mid_eye_l_01_capsule, center=<1.5732 0 0>, rotation=<rot 90 0 0>}]
 dyn_bug_mid_eye_l_base [{shapeType=Capsule, length=0, radius=0, shapeName=dyn_bug_mid_eye_l_base_capsule, center=<-0.026 0 0>, rotation=<rot 90 0 0>}]
 dyn_bug_mid_eye_r_01 [{shapeType=Capsule, length=0.6624, radius=1.6169, shapeName=dyn_bug_mid_eye_r_01_capsule, center=<1.5732 0 0>, rotation=<rot 90 0 0>}]
 dyn_bug_mid_eye_r_base [{shapeType=Capsule, length=0, radius=0, shapeName=dyn_bug_mid_eye_r_base_capsule, rotation=<rot 90 0 0>}]
 dyn_bug_pincer_l_01 [{
   shapeType=Capsule
   length=3.6136
   radius=2.3864
   shapeName=dyn_bug_pincer_l_01_capsule
   center=<3.6234 0.413 1.0226>
   rotation=<rot 63.0228 180 168.2786>
  }]
 dyn_bug_pincer_l_base [{shapeType=Capsule, length=0, radius=0, shapeName=dyn_bug_pincer_l_base_capsule, rotation=<rot 90 0 0>}]
 dyn_bug_pincer_r_01 [{
   shapeType=Capsule
   length=3.6451
   radius=2.3549
   shapeName=dyn_bug_pincer_r_01_capsule
   center=<3.4912 -0.4939 1.0862>
   rotation=<rot 63.0781 -179.999 -171.1911>
  }]
 dyn_bug_pincer_r_base [{shapeType=Capsule, length=0, radius=0, shapeName=dyn_bug_pincer_r_base_capsule, rotation=<rot 90 0 0>}]
 dyn_bug_side_eye_l_01 [{shapeType=Capsule, length=0.6624, radius=1.6169, shapeName=dyn_bug_side_eye_l_01_capsule, center=<1.5732 0 0>, rotation=<rot 90 0 0>}]
 dyn_bug_side_eye_l_base [{shapeType=Capsule, length=0, radius=0, shapeName=dyn_bug_side_eye_l_base_capsule, rotation=<rot 90 0 0>}]
 dyn_bug_side_eye_r_01 [{shapeType=Capsule, length=0.6624, radius=1.6169, shapeName=dyn_bug_side_eye_r_01_capsule, center=<1.5732 0 0>, rotation=<rot 90 0 0>}]
 dyn_bug_side_eye_r_base [{shapeType=Capsule, length=0, radius=0, shapeName=dyn_bug_side_eye_r_base_capsule, rotation=<rot 90 0 0>}]
 dyn_bug_wing_mid_l_01 [{
   shapeType=Capsule
   length=4.5983
   radius=2.6489
   shapeName=dyn_bug_wing_mid_l_01_capsule
   center=<5.178 0 0.9749>
   rotation=<rot 78.2707 -179.9999 -179.9999>
  }]
 dyn_bug_wing_mid_l_base [{shapeType=Capsule, length=0, radius=0, shapeName=dyn_bug_wing_mid_l_base_capsule, rotation=<rot 90 0 0>}]
 dyn_bug_wing_mid_r_01 [{
   shapeType=Capsule
   length=4.5983
   radius=2.6489
   shapeName=dyn_bug_wing_mid_r_01_capsule
   center=<5.1395 0 1.9635>
   rotation=<rot -69.0909 -0.0001 0.0001>
  }]
 dyn_bug_wing_mid_r_base [{shapeType=Capsule, length=0, radius=0, shapeName=dyn_bug_wing_mid_r_base_capsule, rotation=<rot 90 0 0>}]
 dyn_bug_wing_up_l_01 [{
   shapeType=Capsule
   length=7.1475
   radius=2.5217
   shapeName=dyn_bug_wing_up_l_01_capsule
   center=<5.2509 0 1.3707>
   rotation=<rot 69.0647 -179.9885 -179.9855>
  }]
 dyn_bug_wing_up_l_base [{shapeType=Capsule, length=0, radius=0, shapeName=dyn_bug_wing_up_l_base_capsule, rotation=<rot 90 0 0>}]
 dyn_bug_wing_up_r_01 [{
   shapeType=Capsule
   length=7.1475
   radius=2.5217
   shapeName=dyn_bug_wing_up_r_01_capsule
   center=<5.909 0 1.428>
   rotation=<rot -69.0909 -0.0001 0.0001>
  }]
 dyn_bug_wing_up_r_base [{shapeType=Capsule, length=0, radius=0, shapeName=dyn_bug_wing_up_r_base_capsule, rotation=<rot 90 0 0>}]
}
constraints [
 {
  bone1Name=dyn_bug_center_eye_base
  bone2Name=Bug_center
  swing1LimitDegrees=45
  swing1Motion=Free
  swing2LimitDegrees=45
  swing2Motion=Free
  twistLimitDegrees=45
  twistMotion=Free
 }
 {
  bone1Name=dyn_bug_leg_bottom_l_base
  bone2Name=Bug_center
  swing1LimitDegrees=45
  swing1Motion=Free
  swing2LimitDegrees=45
  swing2Motion=Free
  twistLimitDegrees=45
  twistMotion=Free
 }
 {bone1Name=dyn_bug_leg_bottom_l_01, bone2Name=dyn_bug_leg_bottom_l_base, swing1LimitDegrees=50, swing2LimitDegrees=60, twistLimitDegrees=0}
 {
  bone1Name=dyn_bug_leg_bottom_r_base
  bone2Name=Bug_center
  swing1LimitDegrees=45
  swing1Motion=Free
  swing2LimitDegrees=45
  swing2Motion=Free
  twistLimitDegrees=45
  twistMotion=Free
 }
 {bone1Name=dyn_bug_leg_bottom_r_01, bone2Name=dyn_bug_leg_bottom_r_base, swing1LimitDegrees=50, swing2LimitDegrees=60, twistLimitDegrees=0}
 {
  bone1Name=dyn_bug_leg_mid_l_base
  bone2Name=Bug_center
  swing1LimitDegrees=45
  swing1Motion=Free
  swing2LimitDegrees=45
  swing2Motion=Free
  twistLimitDegrees=45
  twistMotion=Free
 }
 {bone1Name=dyn_bug_leg_mid_l_01, bone2Name=dyn_bug_leg_mid_l_base, swing1LimitDegrees=50, swing2LimitDegrees=60, twistLimitDegrees=0}
 {
  bone1Name=dyn_bug_leg_mid_r_base
  bone2Name=Bug_center
  swing1LimitDegrees=45
  swing1Motion=Free
  swing2LimitDegrees=45
  swing2Motion=Free
  twistLimitDegrees=45
  twistMotion=Free
 }
 {bone1Name=dyn_bug_leg_mid_r_01, bone2Name=dyn_bug_leg_mid_r_base, swing1LimitDegrees=50, swing2LimitDegrees=60, twistLimitDegrees=0}
 {
  bone1Name=dyn_bug_leg_up_l_base
  bone2Name=Bug_center
  swing1LimitDegrees=45
  swing1Motion=Free
  swing2LimitDegrees=45
  swing2Motion=Free
  twistLimitDegrees=45
  twistMotion=Free
 }
 {bone1Name=dyn_bug_leg_up_l_01, bone2Name=dyn_bug_leg_up_l_base, swing1LimitDegrees=50, swing2LimitDegrees=60, twistLimitDegrees=0}
 {
  bone1Name=dyn_bug_leg_up_r_base
  bone2Name=Bug_center
  swing1LimitDegrees=45
  swing1Motion=Free
  swing2LimitDegrees=45
  swing2Motion=Free
  twistLimitDegrees=45
  twistMotion=Free
 }
 {bone1Name=dyn_bug_leg_up_r_01, bone2Name=dyn_bug_leg_up_r_base, swing1LimitDegrees=50, swing2LimitDegrees=60, twistLimitDegrees=0}
 {
  bone1Name=dyn_bug_mid_eye_l_base
  bone2Name=Bug_center
  swing1LimitDegrees=45
  swing1Motion=Free
  swing2LimitDegrees=45
  swing2Motion=Free
  twistLimitDegrees=45
  twistMotion=Free
 }
 {bone1Name=dyn_bug_mid_eye_l_01, bone2Name=dyn_bug_mid_eye_l_base, swing1LimitDegrees=40, swing2LimitDegrees=40, twistLimitDegrees=5}
 {
  bone1Name=dyn_bug_mid_eye_r_base
  bone2Name=Bug_center
  swing1LimitDegrees=45
  swing1Motion=Free
  swing2LimitDegrees=45
  swing2Motion=Free
  twistLimitDegrees=45
  twistMotion=Free
 }
 {bone1Name=dyn_bug_mid_eye_r_01, bone2Name=dyn_bug_mid_eye_r_base, swing1LimitDegrees=40, swing2LimitDegrees=40, twistLimitDegrees=5}
 {
  bone1Name=dyn_bug_pincer_l_base
  bone2Name=Bug_center
  swing1LimitDegrees=45
  swing1Motion=Free
  swing2LimitDegrees=45
  swing2Motion=Free
  twistLimitDegrees=45
  twistMotion=Free
 }
 {bone1Name=dyn_bug_pincer_l_01, bone2Name=dyn_bug_pincer_l_base, swing1LimitDegrees=20, swing2LimitDegrees=20, twistLimitDegrees=0}
 {
  bone1Name=dyn_bug_pincer_r_base
  bone2Name=Bug_center
  swing1LimitDegrees=45
  swing1Motion=Free
  swing2LimitDegrees=45
  swing2Motion=Free
  twistLimitDegrees=45
  twistMotion=Free
 }
 {bone1Name=dyn_bug_pincer_r_01, bone2Name=dyn_bug_pincer_r_base, swing1LimitDegrees=20, swing2LimitDegrees=45, twistLimitDegrees=2}
 {
  bone1Name=dyn_bug_side_eye_l_base
  bone2Name=Bug_center
  swing1LimitDegrees=45
  swing1Motion=Free
  swing2LimitDegrees=45
  swing2Motion=Free
  twistLimitDegrees=45
  twistMotion=Free
 }
 {bone1Name=dyn_bug_side_eye_l_01, bone2Name=dyn_bug_side_eye_l_base, swing1LimitDegrees=40, swing2LimitDegrees=40, twistLimitDegrees=5}
 {
  bone1Name=dyn_bug_side_eye_r_base
  bone2Name=Bug_center
  swing1LimitDegrees=45
  swing1Motion=Free
  swing2LimitDegrees=45
  swing2Motion=Free
  twistLimitDegrees=45
  twistMotion=Free
 }
 {bone1Name=dyn_bug_side_eye_r_01, bone2Name=dyn_bug_side_eye_r_base, swing1LimitDegrees=40, swing2LimitDegrees=40, twistLimitDegrees=2}
 {
  bone1Name=dyn_bug_wing_mid_l_base
  bone2Name=Bug_center
  swing1LimitDegrees=45
  swing1Motion=Free
  swing2LimitDegrees=45
  swing2Motion=Free
  twistLimitDegrees=45
  twistMotion=Free
 }
 {bone1Name=dyn_bug_wing_mid_l_01, bone2Name=dyn_bug_wing_mid_l_base, swing1LimitDegrees=40, swing2LimitDegrees=15, twistLimitDegrees=0}
 {
  bone1Name=dyn_bug_wing_mid_r_base
  bone2Name=Bug_center
  swing1LimitDegrees=45
  swing1Motion=Free
  swing2LimitDegrees=45
  swing2Motion=Free
  twistLimitDegrees=45
  twistMotion=Free
 }
 {bone1Name=dyn_bug_wing_mid_r_01, bone2Name=dyn_bug_wing_mid_r_base, swing1LimitDegrees=40, swing2LimitDegrees=15, twistLimitDegrees=0}
 {
  bone1Name=dyn_bug_wing_up_l_base
  bone2Name=Bug_center
  swing1LimitDegrees=45
  swing1Motion=Free
  swing2LimitDegrees=45
  swing2Motion=Free
  twistLimitDegrees=45
  twistMotion=Free
 }
 {bone1Name=dyn_bug_wing_up_l_01, bone2Name=dyn_bug_wing_up_l_base, swing1LimitDegrees=40, swing2LimitDegrees=15, twistLimitDegrees=0}
 {
  bone1Name=dyn_bug_wing_up_r_base
  bone2Name=Bug_center
  swing1LimitDegrees=45
  swing1Motion=Free
  swing2LimitDegrees=45
  swing2Motion=Free
  twistLimitDegrees=45
  twistMotion=Free
 }
 {bone1Name=dyn_bug_wing_up_r_01, bone2Name=dyn_bug_wing_up_r_base, swing1LimitDegrees=40, swing2LimitDegrees=15, twistLimitDegrees=0}
 {bone1Name=dyn_bug_center_eye_01, bone2Name=dyn_bug_center_eye_base, swing1LimitDegrees=25, swing2LimitDegrees=40, twistLimitDegrees=5}
]
