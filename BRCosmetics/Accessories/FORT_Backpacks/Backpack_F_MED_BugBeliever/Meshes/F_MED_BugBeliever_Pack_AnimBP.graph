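# Animation Blueprint for the BugBeliever backpack cosmetic, adding secondary physics motion to the pack and its bug parts.
# It offsets the pack bone, runs base-bone and world-space RigidBody passes, then drives the left pincer from the right, caching this as GooglyEyes.
# A Control Rig derives Auto Eyes from that pose; a bool blend of both feeds AnimDynamics on the zippo bone.
# The update event maps spine bend to a Parashuting value.
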AnimBlueprint /BRCosmetics/Accessories/FORT_Backpacks/Backpack_F_MED_BugBeliever/Meshes/F_MED_BugBeliever_Pack_AnimBP
target_skeleton=&/F_MED_BugBeliever_Pack_Skeleton
graphs [
 AnimGraph {nodes [
   n1 = Use cached pose 'GooglyEyes'()
   n2 = Use cached pose 'Auto Eyes'()
   n3 = Anim Dynamics - Bone: dyn_bug_zippo_01(
    ComponentPose=Local To Component(LocalPose=Blend Poses by bool(BlendPose_0=@n1, BlendPose_1=@n2, BlendTime_1=2) {blendType=Linear})
    Alpha=1
    ExternalForce="0,0,0"
   ) {
    angularDampingOverride=0.7
    angularSpringConstant=50
    bAngularSpring=true
    bOverrideAngularDamping=true
    bUseGravityOverride=true
    simulationSpace=RootRelative
    alphaBoolBlend {blendOption=Linear}
    boundBone {boneName=dyn_bug_zippo_01}
    componentAppliedLinearAccClamp=<3000 3000 3000>
    componentLinearVelScale=<1 2 1.5>
    gravityOverride {z=-980}
    physicsBodyDefinitions [{
      boundBone {boneName=dyn_bug_zippo_01}
      boxExtents=<2 2 2>
      constraintSetup {angularLimitsMax=<5 30 20>, angularLimitsMin=<-5 -5 -120>, angularTarget=<0 0 0>}
      localJointOffset=<0.2 0 0>
     }]
   }
   Output Pose(Result=Component To Local(ComponentPose=@n3)) {layerGroup=None}
   n4 = "Transform (Modify) Bone - Bone: pack"(ComponentPose=Local To Component(LocalPose=Input Pose()), Translation="-6.000,0.339,7.537", Rotation="-0.000,0.000,29.450") {
    rotationMode=BMM_Additive
    rotationSpace=BCS_BoneSpace
    scaleMode=BMM_Additive
    scaleSpace=BCS_BoneSpace
    translationMode=BMM_Additive
    translationSpace=BCS_BoneSpace
    alphaBoolBlend {blendOption=Linear}
    alphaScaleBiasClamp {bInterpResult=true}
    boneToModify {boneName=pack}
    rotation {roll=29.45}
    translation=<-6 0.339 7.537>
   }
   n5 = RigidBody(ComponentPose=@n4) {
    bClampLinearTranslationLimitToRefPose=true
    overridePhysicsAsset=&/F_MED_BugBeliever_Pack_Physics
    simulationSpace=BaseBoneSpace
    alphaBoolBlend {blendOption=Linear}
    alphaScaleBias {bias=1, scale=-1}
    baseBoneRef {boneName=pack}
    componentAppliedLinearAccClamp=<3000 3000 3000>
    simSpaceSettings {maxAngularAcceleration=2200, maxAngularVelocity=6, maxLinearAcceleration=1600, maxLinearVelocity=1200, velocityScaleZ=1.5, worldAlpha=1}
   }
   n6 = RigidBody(ComponentPose=@n5) {
    bClampLinearTranslationLimitToRefPose=true
    overridePhysicsAsset=&F_MED_BugBeliever_Pack_Physics
    simulationSpace=WorldSpace
    alphaBoolBlend {blendOption=Linear}
    alphaScaleBias {scale=0.5}
    baseBoneRef {boneName=pack}
    componentAppliedLinearAccClamp=<1500 1500 1500>
    simSpaceSettings {
     dampingAlpha=3
     maxAngularAcceleration=150
     maxAngularVelocity=2
     maxLinearAcceleration=500
     maxLinearVelocity=1200
     velocityScaleZ=0.85
     worldAlpha=1
    }
   }
   n7 = "dyn_bug_pincer_l_01.rotateY = dyn_bug_pincer_r_01.rotateY - Bone Driven Controller"!(ComponentPose:Component Space Pose Link Structure=@n6, Alpha:float=1) -> (Pose:Component Space Pose Link Structure) {
    is=title AnimGraphNode_BoneDrivenController
    bAffectTargetRotationY=true
    modificationMode=ReplaceComponent
    sourceComponent=RotationY
    alphaBoolBlend {blendOption=Linear}
    sourceBone {boneName=dyn_bug_pincer_r_01}
    targetBone {boneName=dyn_bug_pincer_l_01}
   }
   n8 = "dyn_bug_pincer_l_01.rotateZ = dyn_bug_pincer_r_01.rotateZ - Bone Driven Controller"!(ComponentPose:Component Space Pose Link Structure=@n7, Alpha:float=1) -> (Pose:Component Space Pose Link Structure) {
    is=title AnimGraphNode_BoneDrivenController
    bAffectTargetRotationZ=true
    modificationMode=ReplaceComponent
    sourceComponent=RotationZ
    alphaBoolBlend {blendOption=Linear}
    alphaScaleBiasClamp {scale=-1}
    sourceBone {boneName=dyn_bug_pincer_r_01}
    targetBone {boneName=dyn_bug_pincer_l_01}
   }
   n9 = "dyn_bug_pincer_l_01.rotateX = dyn_bug_pincer_r_01.rotateX - Bone Driven Controller"!(ComponentPose:Component Space Pose Link Structure=@n8, Alpha:float=1) -> (Pose:Component Space Pose Link Structure) {
    is=title AnimGraphNode_BoneDrivenController
    bAffectTargetRotationX=true
    modificationMode=ReplaceComponent
    sourceComponent=RotationX
    alphaBoolBlend {blendOption=Linear}
    alphaScaleBiasClamp {scale=-1}
    sourceBone {boneName=dyn_bug_pincer_r_01}
    targetBone {boneName=dyn_bug_pincer_l_01}
   }
   n10 = "Transform (Modify) Bone - Bone: dyn_bug_pincer_l_01"!(ComponentPose=@n9, Alpha=1) -> (Pose) {
    rotationMode=BMM_Additive
    rotationSpace=BCS_BoneSpace
    alphaBoolBlend {blendOption=Linear}
    boneToModify {boneName=dyn_bug_pincer_l_01}
    rotation {yaw=-20}
   }
   Save cached pose 'GooglyEyes'(Pose=Component To Local(ComponentPose=@n10)) {cacheName=GooglyEyes}
   n11 = Use cached pose 'GooglyEyes'()
   n12 = Control Rig(Source=Component To Local(ComponentPose=Local To Component(LocalPose=@n11))) {
    alphaBoolBlend {blendOption=Linear}
    controlRigAssetReference {blueprintRigClass=&/F_MED_BugBeliever_Pack_CtrlRig.F_MED_BugBeliever_Pack_CtrlRig_C}
   }
   Save cached pose 'Auto Eyes'(Pose=@n12) {cacheName=Auto Eyes}
   NOT Boolean(A=Get bIsInFrontEnd())
  ]}
 EventGraph {nodes [
   Event BlueprintUpdateAnimation()
   > n1 = Sequence()
   n2 = GetDeltaTransformFromRefPose!(self=GetOwnerSkeletalMeshComponent(), BoneName=spine_05, BaseName=spine_01) -> (ReturnValue_Location:Vector, ReturnValue_Rotation:Rotator, ReturnValue_Scale:Vector)
   Set Parashuting(execute=@n1.then_1, Parashuting:double=MapRangeClamped(Value=BreakRotator(InRot=@n2.ReturnValue_Rotation).Yaw, InRangeA=-25, InRangeB=-65)) -> (Output_Get:double)
   TryGetPawnOwner()
  ]}
]
variables [
 Gravity Override {type=Vector}
 Gravity Override Params {type=GravityOverrideParamsStruct}
 Parashuting {type=number}
]
